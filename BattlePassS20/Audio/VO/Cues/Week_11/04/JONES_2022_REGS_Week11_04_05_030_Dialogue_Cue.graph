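# Sound cue named for a Jones dialogue line in the Season 20 Battle Pass, Week 11 voice-over set.
# It routes to the Fort_VO_Narrator sound class with a raised volume multiplier and a very large max distance.
# The cue graph has no root or nodes, so it references no audio in this file.
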SoundCue /BattlePassS20/Audio/VO/Cues/Week_11/04/JONES_2022_REGS_Week11_04_05_030_Dialogue_Cue
root=null
nodes []
properties {maxDistance=2097152, soundClassObject=/Game/Sounds/FortSoundClassesAndModes/Fort_VO_Narrator, volumeMultiplier=1.3333}
